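# Physics asset for the Tactical Scuba backpack cosmetic, defining simulated bodies for its dangling parts.
# Capsule bodies cover the pack, a hose, a needle and left and right knives.
# Each small part constrains directly to the pack body, with limited swing and twist for the hose and knives.
# The needle pivots on a single swing axis, with its other swing and its twist locked.
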PhysicsAsset /BRCosmetics/Accessories/FORT_Backpacks/Backpack_M_MED_Tactical_Scuba/Meshes/M_MED_Tactical_Scuba_Pack_Physics
bodies {
 dyn_hose [{shapeType=Capsule, length=0, radius=1.7552, center=<4.1701 0.3256 0>, rotation=<rot 0 -90 0>}]
 dyn_knife_l [{shapeType=Capsule, length=13.1949, radius=2.4103, center=<2.0545 0 0.302>, rotation=<rot 90 0 0>}]
 dyn_knife_r [{shapeType=Capsule, length=14.682, radius=2.752, center=<-1.3676 0 -1.0735>, rotation=<rot 87.9882 -0.0004 -0.0004>}]
 dyn_needle [{shapeType=Capsule, length=0.505, radius=0.505, center=<0.5 0 0>, rotation=<rot 0 -90 -88.7475>}]
 pack [{shapeType=Capsule, length=24.1984, radius=7, center=<0 -8.5484 -7.9989>}]
}
constraints [
 {bone1Name=dyn_knife_l, bone2Name=pack, swing1LimitDegrees=30, swing2LimitDegrees=30, twistLimitDegrees=5}
 {bone1Name=dyn_knife_r, bone2Name=pack, swing1LimitDegrees=30, swing2LimitDegrees=30, twistLimitDegrees=5}
 {bone1Name=dyn_hose, bone2Name=pack, swing1LimitDegrees=45, swing2LimitDegrees=45, twistLimitDegrees=15}
 {
  bone1Name=dyn_needle
  bone2Name=pack
  swing1LimitDegrees=20
  swing2LimitDegrees=0
  swing2Motion=Locked
  twistLimitDegrees=0
  twistMotion=Locked
 }
]
